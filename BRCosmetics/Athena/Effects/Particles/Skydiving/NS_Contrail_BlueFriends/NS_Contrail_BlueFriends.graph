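# Niagara skydiving contrail cosmetic "BlueFriends" that surrounds the player with water meshes, splash sprites and fish.
# ParaGlideLeanAlpha crossfades calm and dive emitters, with dive layers ramping by its cube.
# Calm fish follow analytic parabolic leaps facing their motion, while dive fish stream from a bone against the FN_Contrail_Setup direction.
# Main splash emitters use two spawn groups with platform-dependent rates; a front-end preview flag adjusts dive offsets and sizing.
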NiagaraSystem /BRCosmetics/Athena/Effects/Particles/Skydiving/NS_Contrail_BlueFriends/NS_Contrail_BlueFriends
dependencies {
 usedDataInterfaces [
  &NiagaraDataInterfaceCurve
  &NiagaraDataInterfaceSpriteRendererInfo
  &NiagaraDataInterfaceMeshRendererInfo
  &NiagaraDataInterfacePlatformSet
  &NiagaraDataInterfaceArrayFloat
  &NiagaraDataInterfaceVector2DCurve
  &NiagaraDataInterfaceArrayFloat3
  &NiagaraDataInterfaceSkeletalMesh
 ]
 usedDynamicInputs []
 usedModules [
  &self:SystemSpawnScript.SystemScriptSource.SystemScriptGraph.NiagaraNodeAssignment_0.SetVariables_#1
  &SystemState
  /Game/Effects/Niagara/Modules/Contrails/FN_Contrail_Setup
  &FN_Integrator
 ]
}
emitters {
 FishCalm {
  emitterUpdateScript [
   &EmitterState(Life Cycle Mode=::NewEnumerator1, Loop Behavior=::NewEnumerator1, Loop Delay=1, Loop Duration Mode=::NewEnumerator1, UseLoopDelay=true)
   &SpawnRate(SpawnRate=&Lerp_Float(A=10, B=0, Alpha=$User.ParaGlideLeanAlpha), Use Spawn Probability=false)
  ]
  particleSpawnScript [
   &InitializeParticle(
    Lifetime=&RandomRangeFloat(Minimum=0.5, Maximum=0.7)
    Mesh Scale Mode=::NewEnumerator3
    Mesh Uniform Scale=&RandomRangeFloat(Minimum=6, Maximum=10)
    Sprite Size Mode=::NewEnumerator3
    Uniform Sprite Size=100
    UsePositionOffset=false
    Write Material Random=true
   )
   /SetVariables(Particles.LastPos:Vector3f=<0 -200 0>) {isSetParameters=true}
   /SetVariables(Particles.SpawnPos:Vector3f=&RandomRangeVector(Minimum=<50 0 0>, Maximum=<-50 -50 1>)) {isSetParameters=true}
  ]
  particleUpdateScript [
   &ParticleState()
   /SetVariables(Particles.RandomFish:NiagaraInt32=&RandomRangeInt(Minimum=0, Maximum=1)) {isSetParameters=true}
   &SolveForcesAndVelocity(Clamp Velocity=false, Limit Acceleration=false)
   /SetVariables(Particles.Pos:Vector3f=&Add_Vector(
     A=&Lerp_Vector(
      A=$Particles.SpawnPos
      B=&Multiply_Vector(
       A=&Add_Vector(
        A=$Particles.SpawnPos
        B=&Multiply_VectorByFloat(
         Vector=&NormalizeVector(
          Vector To Normalize=&Add_Vector(
           A=&RandomVector(Vector Scale=1, Constrain to Hemisphere=true, Hemisphere Constraint Vector=<0 1 0>)
           B=<0 0.5 0>
          )
          Safe Normalize=true
          Safe Fallback Threshold=0
          Safe Fallback Vector=<0 0 0.001>
         )
         Float=&RandomRangeFloat(Minimum=135, Maximum=175)
        )
       )
       B=<1 1 0>
      )
      Alpha=$Particles.NormalizedAge
     )
     B=&MakeVector(
      X=0
      Y=0
      Z=&Multiply_Float(
       A=&Multiply_Float(A=&RandomRangeFloat(Minimum=25, Maximum=50), B=4)
       B=&Multiply_Float(A=$Particles.NormalizedAge, B=&OneMinusFloat(Float=$Particles.NormalizedAge))
      )
     )
    )) {isSetParameters=true}
   /SetVariables(Particles.Dir:Vector3f=&NormalizeVector(
     Vector To Normalize=&Subtract_Vector(A=$Particles.Pos, B=$Particles.LastPos)
     Safe Normalize=true
     Safe Fallback Threshold=0
     Safe Fallback Vector=<0 0 0.001>
    )) {isSetParameters=true}
   /SetVariables(Particles.LastPos:Vector3f=$Particles.Pos) {isSetParameters=true}
   /SetVariables(Particles.Position=&AddVectorToPosition(Position=$Engine.Emitter.SimulationPosition, Vector=$Particles.Pos)) {isSetParameters=true}
   &UpdateMeshOrientation(
    Facing Direction=$Particles.Dir
    Facing Interpolation=false
    Orientation Method=::NewEnumerator1
    Override Quat=false
    TransformFacing=false
    TransformUp=false
   )
   &ScaleMeshSize(Scale Factor=&VectorFromFloat(Value=&FloatFromCurve(FloatCurve=DataInterface() {propertyValues {Curve {
         defaultValue=1
         keys [
          {interpMode=RCIM_Cubic, tangentMode=RCTM_User, time=0, value=0}
          {arriveTangent=2, leaveTangent=2, time=0.1, value=1}
          {interpMode=RCIM_Cubic, tangentMode=RCTM_User, time=0.8, value=0.9}
          {tangentMode=RCTM_User, time=1, value=0}
         ]
        }}})))
   &self:ScratchModule(Delay=0.2) {module=ScratchModule}
   &DynamicMaterialParameters(
    Index 0 Param 1=&RandomRangeFloat(Minimum=0.25, Maximum=1)
    Index 0 Param 2=&RandomRangeFloat(Minimum=0.5, Maximum=1)
    Index 1 Param 1=&RandomRangeFloat(Minimum=0, Maximum=1)
    Index 1 Param 2=50
    Module.Index0Float0Write=true
    Module.Index0Float1Write=true
    Module.Index0Float2Write=false
    Module.Index0Float3Write=false
    Module.Index1Float0Write=false
    Module.Index1Float1Write=false
    Module.Index1Float2Write=false
    Module.Index1Float3Write=false
    Write Parameter Index 1=::NewEnumerator1
   )
  ]
  rendererClasses [&NiagaraMeshRendererProperties]
  renderers [
   {rendererClass=&NiagaraMeshRendererProperties, rendererIndex=0}
   {rendererClass=&NiagaraMeshRendererProperties, rendererIndex=1}
  ]
 }
 FishDive {
  emitterUpdateScript [
   &EmitterState(Delay First Loop Only=true, Life Cycle Mode=::NewEnumerator1, Loop Delay=1, Loop Duration=1, UseLoopDelay=true)
   &SpawnRate(
    SpawnRate=&Multiply_Float(A=&Power(a=$User.ParaGlideLeanAlpha, b=3), B=10)
    Use Spawn Probability=false
   )
  ]
  particleSpawnScript [
   &InitializeParticle(
    Color Mode=::NewEnumerator0
    Lifetime=1.5
    Mesh Scale Mode=::NewEnumerator1
    Mesh Uniform Scale Max=6
    Mesh Uniform Scale Min=2
    UsePositionOffset=false
    Write Material Random=true
   )
   &SkeletalMeshLocation(
    Bone / Socket Index=6
    Bone Sampling Mode=::NewEnumerator4
    MaskSpawnGroup=false
    Sampled Position Offset=/Niagara/DynamicInputs/TypeConversions/MakeCustomPositionFromBool(
     Bool:NiagaraBool=$User.bIsFrontEndPreview
     True Position:NiagaraPosition=&RandomRangeVector(Minimum=<-75 -150 -200>, Maximum=<75 -25 -200>)
     False Position:NiagaraPosition=<0 -25 -200>
    )
    Sampled Position Offset Coordinate Space:ENiagaraSkelMeshTransforms=::NewEnumerator2
    Skeletal Mesh=$System.Skeletal Mesh
   )
   &InitialMeshOrientation(Use Rotation Vector=false)
   &AddVelocity(
    Cone Angle=30
    Cone Axis=/Niagara/DynamicInputs/Vectors/NegateVector(Vector:Vector3f=$System.FN_Contrail_Setup.Direction Vector)
    Inner Cone Angle=10
    Module.UseConeFalloff=false
    Module.UseVelDistribution=false
    Override Seed=false
    Velocity Mode=::NewEnumerator2
    Velocity Speed=&RandomRangeFloat(Minimum=400, Maximum=500)
   ) {module=AddVelocity001}
  ]
  particleUpdateScript [
   &ParticleState()
   /SetVariables(Particles.RandomFish:NiagaraInt32=&RandomRangeInt(Minimum=0, Maximum=1)) {isSetParameters=true}
   &ScaleMeshSize(Scale Factor=&Multiply_Vector(
     A=<1 1 1>
     B=&VectorFromFloat(Value=&FloatFromCurve(FloatCurve=DataInterface() {propertyValues {Curve {
          defaultValue=1
          keys [
           {interpMode=RCIM_Cubic, tangentMode=RCTM_User, time=0, value=0}
           {arriveTangentWeight=0.0707, leaveTangentWeight=0.4007, time=0.15, value=1}
           {arriveTangent=-1.1765, interpMode=RCIM_Cubic, leaveTangent=-1.1765, time=0.8, value=1}
           {interpMode=RCIM_Cubic, tangentMode=RCTM_User, time=1, value=0}
          ]
         }}}))
    ))
   &UpdateMeshOrientation(
    Rotation Rate=$Engine.DeltaTime
    Rotation Vector=&Multiply_VectorByFloat(
     Vector=<0 1 1>
     Float=&FloatFromCurve(
      FloatCurve=DataInterface() {propertyValues {Curve {
         defaultValue=1
         keys [
          {arriveTangent=-2.6167, interpMode=RCIM_Cubic, leaveTangent=-2.6167, tangentMode=RCTM_User, time=0, value=1}
          {tangentMode=RCTM_User, time=1, value=0}
         ]
        }}}
      Scale Curve=100
     )
    )
   )
   &ScaleVelocity(Velocity Scale=&MakeVector(
     X=1
     Y=&FloatFromCurve(FloatCurve=DataInterface() {propertyValues {Curve {
         defaultValue=1
         keys [
          {interpMode=RCIM_Cubic, leaveTangentWeight=0.3333, tangentMode=RCTM_User, time=0, value=0.301}
          {arriveTangent=0.199, interpMode=RCIM_Cubic, leaveTangent=0.199, time=0.445, value=-0.123}
          {arriveTangentWeight=0.3333, interpMode=RCIM_Cubic, tangentMode=RCTM_User, time=1, value=0.5}
         ]
        }}})
     Z=1
    ))
   &SolveForcesAndVelocity(Clamp Velocity=false, Limit Acceleration=false)
   &DynamicMaterialParameters(
    Index 0 Param 1=&RandomRangeFloat(Minimum=0.25, Maximum=1)
    Index 0 Param 2=&RandomRangeFloat(Minimum=0.5, Maximum=1)
    Module.Index0Float0Write=true
    Module.Index0Float1Write=true
    Module.Index0Float2Write=false
    Module.Index0Float3Write=false
   )
  ]
  rendererClasses [&NiagaraMeshRendererProperties]
  renderers [
   {rendererClass=&NiagaraMeshRendererProperties, rendererIndex=0}
   {rendererClass=&NiagaraMeshRendererProperties, rendererIndex=1}
  ]
 }
 Ripples {
  emitterUpdateScript [&EmitterState()]
  particleSpawnScript [
   &InitializeParticle(
    Lifetime=&RandomRangeFloat(Minimum=0.5, Maximum=0.6)
    Sprite Rotation Mode=::NewEnumerator2
    Sprite Size Mode=::NewEnumerator1
    Sprite UV Mode=::NewEnumerator3
    Uniform Sprite Size Max=50
    Uniform Sprite Size Min=25
    UsePositionOffset=false
    Write Material Random=true
   )
   &AddVelocity(Velocity=<0 -1 0>, Velocity Speed Scale=250)
  ]
  particleUpdateScript [
   &ParticleState()
   &ScaleSpriteSize(Uniform Curve Sprite Scale=DataInterface() {propertyValues {Curve {
       defaultValue=0
       keys [
        {arriveTangent=1.5815, interpMode=RCIM_Cubic, leaveTangent=1.5815, tangentMode=RCTM_User, time=0, value=0}
        {interpMode=RCIM_Cubic, tangentMode=RCTM_User, time=1, value=1}
       ]
      }}})
   &ScaleColor(
    Scale Alpha=&FloatFromCurve(FloatCurve=DataInterface() {propertyValues {Curve {
        defaultValue=0
        keys [
         {interpMode=RCIM_Cubic, tangentMode=RCTM_User, time=0, value=0}
         {arriveTangent=0.2882, interpMode=RCIM_Cubic, leaveTangent=0.2882, tangentMode=RCTM_User, time=0.05, value=1}
         {tangentMode=RCTM_User, time=1, value=0}
        ]
       }}})
    ScaleA=true
    ScaleRGB=false
   )
   &SpriteFacingAndAlignment(Sprite Facing=<0 0 1>, WriteAlignment=false, WriteFacing=true)
   &SolveForcesAndVelocity(Clamp Velocity=false, Limit Acceleration=false)
  ]
  rendererClasses [&NiagaraSpriteRendererProperties]
  renderers [{rendererClass=&NiagaraSpriteRendererProperties}]
 }
 Splashes {
  emitterUpdateScript [&EmitterState()]
  particleSpawnScript [
   &InitializeParticle(
    Color=<rgba 1 1 1 0.5>
    Lifetime=&RandomRangeFloat(Minimum=0.3, Maximum=0.4)
    Sprite Rotation Mode=::NewEnumerator2
    Sprite Size Mode=::NewEnumerator1
    Sprite UV Mode=::NewEnumerator3
    Uniform Sprite Size Max=60
    Uniform Sprite Size Min=80
    UsePositionOffset=false
    Write Material Random=true
   )
   &SubUVAnimation(
    Evaluation Type=::NewEnumerator0
    "Sprite Renderer "=DataInterface() {propertyValues {SpriteRenderer=&self:Splashes_0.NiagaraSpriteRendererProperties_0}}
    SubUV Animation Mode:ENiagara_SubUVLookupModeV2=::NewEnumerator2
    UseEndFrame=false
    UseStartFrame=false
   )
   &AddVelocity(Velocity=<0 -1 1>, Velocity Speed Scale=&RandomRangeFloat(Minimum=200, Maximum=350))
  ]
  particleUpdateScript [
   &ParticleState()
   &ScaleSpriteSize(Uniform Curve Sprite Scale=DataInterface() {propertyValues {Curve {
       defaultValue=0
       keys [
        {arriveTangent=1.7051, interpMode=RCIM_Cubic, leaveTangent=1.7051, tangentMode=RCTM_User, time=0, value=0}
        {interpMode=RCIM_Cubic, tangentMode=RCTM_User, time=1, value=1}
       ]
      }}})
   &Drag(Use Linear Drag=true, Use Rotational Drag=true)
   &GravityForce()
   &ScaleColor(
    Scale Alpha=&FloatFromCurve(FloatCurve=DataInterface() {propertyValues {Curve {
        defaultValue=1
        keys [
         {interpMode=RCIM_Cubic, tangentMode=RCTM_User, time=0, value=1}
         {arriveTangent=-3.0312, interpMode=RCIM_Cubic, leaveTangent=-3.0312, tangentMode=RCTM_User, time=1, value=0}
        ]
       }}})
    ScaleA=true
    ScaleRGB=false
   )
   &DynamicMaterialParameters(
    Index 0 Param 3=&FloatFromCurve(FloatCurve=DataInterface() {propertyValues {Curve {
        defaultValue=1
        keys [{interpMode=RCIM_Cubic, tangentMode=RCTM_User, time=0, value=0}, {interpMode=RCIM_Cubic, tangentMode=RCTM_User, time=1, value=1}]
       }}})
    Index 0 Param 4=&FloatFromCurve(FloatCurve=DataInterface() {propertyValues {Curve {
        defaultValue=1
        keys [
         {interpMode=RCIM_Cubic, tangentMode=RCTM_User, time=0, value=0.6}
         {arriveTangent=1.2977, interpMode=RCIM_Cubic, leaveTangent=1.2977, tangentMode=RCTM_User, time=1, value=1}
        ]
       }}})
    Module.Index0Float0Write=false
    Module.Index0Float1Write=false
    Module.Index0Float2Write=true
    Module.Index0Float3Write=true
   )
   &SolveForcesAndVelocity(Clamp Velocity=false, Limit Acceleration=false)
  ]
  rendererClasses [&NiagaraSpriteRendererProperties]
  renderers [{rendererClass=&NiagaraSpriteRendererProperties}]
 }
 WaterCalm {
  emitterUpdateScript [
   &EmitterState(Life Cycle Mode=::NewEnumerator1, Loop Behavior=::NewEnumerator1, Loop Delay=1, Loop Duration Mode=::NewEnumerator1, UseLoopDelay=true)
   &SpawnBurst_Instantaneous(Use Spawn Probability=false, UseLoopCountLimit=false)
  ]
  particleSpawnScript [
   /SetVariables(Particles.Color_2:LinearColor=<rgba 1 1 1 0.4>) {isSetParameters=true}
   /SetVariables(Particles.Color_3:LinearColor=<rgba 0.01 0.38 0.95 1>) {isSetParameters=true}
   &InitializeParticle(Color=<rgba 0.85 0.95 1 0.25>, Mesh Scale Mode=::NewEnumerator4, UsePositionOffset=false, Write Material Random=true)
  ]
  particleUpdateScript [
   &ParticleState(Kill Particles When Lifetime Has Elapsed=false)
   &ScaleMeshSize(Scale Factor=&Lerp_Vector(A=<1 1 1>, B=<1 0.25 1>, Alpha=$User.ParaGlideLeanAlpha))
   &SkeletalMeshLocation(
    Bone / Socket Index=4
    Bone Sampling Mode=::NewEnumerator4
    MaskSpawnGroup=false
    Orientation Sampling=::NewEnumerator1
    Sampled Position Offset=<0 -35 -75>
    Sampled Position Offset Coordinate Space:ENiagaraSkelMeshTransforms=::NewEnumerator2
    Skeletal Mesh=$System.Skeletal Mesh
   )
   /SetVariables(Particles.Color_2:LinearColor=&MakeLinearColorFromVectorAndFloat(
     "Vector (RGB)"=<1 1 1>
     "Float (Alpha)"=&Lerp_Float(
      A=&FloatFromCurve(FloatCurve=DataInterface() {propertyValues {Curve {
          defaultValue=1
          keys [{interpMode=RCIM_Cubic, tangentMode=RCTM_User, time=0, value=0}, {interpMode=RCIM_Cubic, tangentMode=RCTM_User, time=1, value=1}]
         }}})
      B=0
      Alpha=$User.ParaGlideLeanAlpha
     )
    )) {isSetParameters=true}
   /SetVariables(Particles.Color_3:LinearColor=&MakeLinearColorFromVectorAndFloat(
     "Vector (RGB)"=<0.01 0.38 0.95>
     "Float (Alpha)"=&Lerp_Float(
      A=&FloatFromCurve(
       FloatCurve=DataInterface() {propertyValues {Curve {
          defaultValue=1
          keys [{interpMode=RCIM_Cubic, tangentMode=RCTM_User, time=0, value=0}, {interpMode=RCIM_Cubic, tangentMode=RCTM_User, time=1, value=1}]
         }}}
       Scale Curve=0.65
      )
      B=0
      Alpha=$User.ParaGlideLeanAlpha
     )
    )) {isSetParameters=true}
   &ScaleColor(
    Scale Alpha=&Lerp_Float(
     A=&FloatFromCurve(FloatCurve=DataInterface() {propertyValues {Curve {
         defaultValue=1
         keys [{interpMode=RCIM_Cubic, tangentMode=RCTM_User, time=0, value=0}, {interpMode=RCIM_Cubic, tangentMode=RCTM_User, time=1, value=1}]
        }}})
     B=0
     Alpha=$User.ParaGlideLeanAlpha
    )
    ScaleA=true
    ScaleRGB=false
   )
   &DynamicMaterialParameters(
    Index 0 Param 2=0.2
    Module.Index0Float0Write=false
    Module.Index0Float1Write=true
    Module.Index0Float2Write=false
    Module.Index0Float3Write=false
   )
  ]
  rendererClasses [&NiagaraMeshRendererProperties]
  renderers [
   {rendererClass=&NiagaraMeshRendererProperties, rendererIndex=0}
   {rendererClass=&NiagaraMeshRendererProperties, rendererIndex=1}
   {rendererClass=&NiagaraMeshRendererProperties, rendererIndex=2}
   {rendererClass=&NiagaraMeshRendererProperties, rendererIndex=3}
  ]
 }
 WaterDive {
  emitterUpdateScript [
   &EmitterState(Life Cycle Mode=::NewEnumerator1, Loop Behavior=::NewEnumerator1, Loop Delay=1, Loop Duration Mode=::NewEnumerator1, UseLoopDelay=true)
   &SpawnBurst_Instantaneous(Use Spawn Probability=false, UseLoopCountLimit=false)
  ]
  particleSpawnScript [
   /SetVariables(Particles.Color_2:LinearColor=<rgba 0.01 0.38 0.95 1>) {isSetParameters=true}
   &InitializeParticle(Color=<rgba 0.85 0.95 1 0.5>, Mesh Scale=<1 1 3>, Mesh Scale Mode=::NewEnumerator4, UsePositionOffset=false, Write Material Random=true)
  ]
  particleUpdateScript [
   &ParticleState(Kill Particles When Lifetime Has Elapsed=false)
   &SkeletalMeshLocation(
    Bone / Socket Index=6
    Bone Sampling Mode=::NewEnumerator4
    MaskSpawnGroup=false
    Sampled Position Offset=&MakeCustomPositionFromBool(Bool:NiagaraBool=$User.bIsFrontEndPreview, True Position:NiagaraPosition=<0 50 -50>, False Position:NiagaraPosition=<0 0 -10>)
    Sampled Position Offset Coordinate Space:ENiagaraSkelMeshTransforms=::NewEnumerator2
    Skeletal Mesh=$System.Skeletal Mesh
   )
   /SetVariables(Particles.Color_2:LinearColor=&MakeLinearColorFromVectorAndFloat(
     "Vector (RGB)"=<0.01 0.38 0.95>
     "Float (Alpha)"=&Multiply_Float(A=&Power(a=$User.ParaGlideLeanAlpha, b=3), B=1)
    )) {isSetParameters=true}
   &ScaleColor(
    Scale Alpha=&Multiply_Float(A=&Power(a=$User.ParaGlideLeanAlpha, b=3), B=1)
    ScaleA=true
    ScaleRGB=false
   )
   &DynamicMaterialParameters(
    Index 0 Param 2=0.6
    Module.Index0Float0Write=false
    Module.Index0Float1Write=true
    Module.Index0Float2Write=false
    Module.Index0Float3Write=false
   )
  ]
  rendererClasses [&NiagaraMeshRendererProperties]
  renderers [
   {rendererClass=&NiagaraMeshRendererProperties, rendererIndex=0}
   {rendererClass=&NiagaraMeshRendererProperties, rendererIndex=1}
   {rendererClass=&NiagaraMeshRendererProperties, rendererIndex=2}
  ]
 }
 WaterSplashes {
  emitterUpdateScript [
   &EmitterState(Delay First Loop Only=true, Life Cycle Mode=::NewEnumerator1, Loop Delay=1, Loop Duration=1, UseLoopDelay=true)
   &SpawnRate(
    SpawnRate=&Lerp_Float(
     A=&Multiply_Float(A=&FN_PickFloatDependOnPlatform(TargetPlatformActive=50, TargetPlatformNoActive=25), B=1)
     B=0
     Alpha=$User.ParaGlideLeanAlpha
    )
    Use Spawn Probability=false
   )
   &SpawnRate(
    SpawnGroup=1
    SpawnRate=&Lerp_Float(
     A=&FN_PickFloatDependOnPlatform(TargetPlatformActive=50, TargetPlatformNoActive=25)
     B=0
     Alpha=$User.ParaGlideLeanAlpha
    )
    Use Spawn Probability=false
   ) {module=SpawnRate001}
  ]
  particleSpawnScript [
   &InitializeParticle(
    Lifetime=&RandomRangeFloat(
     Minimum=&SelectFloatFromArray(Direct Array Index=&SpawnGroup(), Float Selection Array=DataInterface() {propertyValues {FloatData [0.4,0.3]}})
     Maximum=&SelectFloatFromArray(Direct Array Index=&SpawnGroup(), Float Selection Array=DataInterface() {propertyValues {FloatData [0.6,0.4]}})
    )
    Sprite Rotation Mode=::NewEnumerator2
    Sprite Size Mode=::NewEnumerator1
    Sprite UV Mode=::NewEnumerator3
    Uniform Sprite Size Max=100
    Uniform Sprite Size Min=75
    UsePositionOffset=false
    Write Material Random=true
   )
   &SubUVAnimation(
    Evaluation Type=::NewEnumerator0
    "Sprite Renderer "=DataInterface() {propertyValues {SpriteRenderer=&self:WaterSplashes_1.NiagaraSpriteRendererProperties_0}}
    SubUV Animation Mode:ENiagara_SubUVLookupModeV2=::NewEnumerator2
    UseEndFrame=false
    UseStartFrame=false
   )
   &SystemLocation(Offset=&RandomRangeVector(
     Minimum=&SelectVectorFromArray(
      Direct Array Index=&SpawnGroup()
      Vector Selection Array=DataInterface() {propertyValues {FloatData [<65 125 -20>, <75 125 -20>]}}
     )
     Maximum=&SelectVectorFromArray(
      Direct Array Index=&SpawnGroup()
      Vector Selection Array=DataInterface() {propertyValues {FloatData [<-65 100 10>, <-75 100 10>]}}
     )
    ))
   &AddVelocity(
    Velocity=&SelectVectorFromArray(
     Direct Array Index=&SpawnGroup()
     Vector Selection Array=DataInterface() {propertyValues {FloatData [<0 -1 0.5>, <0.1 -0.6 1>]}}
    )
    Velocity Speed Scale=&RandomRangeFloat(
     Minimum=&SelectFloatFromArray(Direct Array Index=&SpawnGroup(), Float Selection Array=DataInterface() {propertyValues {FloatData [100,300]}})
     Maximum=&SelectFloatFromArray(Direct Array Index=&SpawnGroup(), Float Selection Array=DataInterface() {propertyValues {FloatData [350,400]}})
    )
   )
  ]
  particleUpdateScript [
   &ParticleState()
   /SetVariables(Particles.SubSize:Vector2f=&Multiply_Vector2D(
     A=&Multiply_Vector2D_ByFloat(Float=&RandomRangeFloat(Minimum=2, Maximum=3), Vector2D=<2 2>)
     B=&Vector2DFromCurve(
      Vector2Curve=DataInterface() {propertyValues {
        XCurve {
         defaultValue=1
         keys [
          {interpMode=RCIM_Cubic, tangentMode=RCTM_User, time=0, value=1}
          {arriveTangent=-1.792, interpMode=RCIM_Cubic, leaveTangent=-1.792, tangentMode=RCTM_User, time=1, value=0}
         ]
        }
        YCurve {
         defaultValue=1
         keys [
          {arriveTangent=0.1106, interpMode=RCIM_Cubic, leaveTangent=0.1106, tangentMode=RCTM_User, time=0, value=0}
          {arriveTangent=-0.0577, interpMode=RCIM_Cubic, leaveTangent=-0.0577, tangentMode=RCTM_User, time=0.15, value=4}
          {arriveTangent=-8.5368, interpMode=RCIM_Cubic, leaveTangent=-8.5368, tangentMode=RCTM_User, time=0.7403, value=1.8683}
          {arriveTangent=-0.0405, interpMode=RCIM_Cubic, leaveTangent=-0.0405, tangentMode=RCTM_User, time=1, value=0}
         ]
        }
       }}
      ScaleCurve=<1 1>
     )
    )) {isSetParameters=true, module=SetVariables_#2}
   /SetVariables(Particles.VisibilityTag=&SpawnGroup()) {isSetParameters=true, module=SetVariables_#3}
   &ScaleSpriteSize(Uniform Curve Sprite Scale=DataInterface() {propertyValues {Curve {
       defaultValue=0
       keys [
        {arriveTangent=1.1866, interpMode=RCIM_Cubic, leaveTangent=1.1866, tangentMode=RCTM_User, time=0, value=0}
        {interpMode=RCIM_Cubic, tangentMode=RCTM_User, time=1, value=1}
       ]
      }}})
   &Drag(Use Linear Drag=true, Use Rotational Drag=true)
   &ScaleColor(
    Scale Alpha=&FloatFromCurve(FloatCurve=DataInterface() {propertyValues {Curve {
        defaultValue=1
        keys [
         {arriveTangent=-0.1688, interpMode=RCIM_Cubic, leaveTangent=-0.1688, tangentMode=RCTM_User, time=0, value=1}
         {arriveTangent=-3.0312, interpMode=RCIM_Cubic, leaveTangent=-3.0312, tangentMode=RCTM_User, time=1, value=0}
        ]
       }}})
    ScaleA=true
    ScaleRGB=false
   )
   &DynamicMaterialParameters(
    Index 0 Param 3=&FloatFromCurve(FloatCurve=DataInterface() {propertyValues {Curve {
        defaultValue=1
        keys [{interpMode=RCIM_Cubic, tangentMode=RCTM_User, time=0, value=0}, {interpMode=RCIM_Cubic, tangentMode=RCTM_User, time=1, value=1}]
       }}})
    Index 0 Param 4=&FloatFromCurve(FloatCurve=DataInterface() {propertyValues {Curve {
        defaultValue=1
        keys [
         {interpMode=RCIM_Cubic, tangentMode=RCTM_User, time=0, value=0.6}
         {arriveTangent=1.2977, interpMode=RCIM_Cubic, leaveTangent=1.2977, tangentMode=RCTM_User, time=1, value=1}
        ]
       }}})
    Module.Index0Float0Write=false
    Module.Index0Float1Write=false
    Module.Index0Float2Write=true
    Module.Index0Float3Write=true
   )
   &GravityForce(Gravity=&SelectVectorFromArray(
     Direct Array Index=&SpawnGroup()
     Vector Selection Array=DataInterface() {propertyValues {FloatData [<0 0 0>, <0 0 -980>]}}
    ))
   &SolveForcesAndVelocity(Clamp Velocity=false, Limit Acceleration=false)
   &ScaleSpriteSizeBySpeed(Min Scale Factor=<0.5 0.5>)
  ]
  rendererClasses [&NiagaraSpriteRendererProperties]
  renderers [
   {rendererClass=&NiagaraSpriteRendererProperties}
   {rendererClass=&NiagaraSpriteRendererProperties, rendererIndex=1}
  ]
 }
 WaterSplashesDive {
  emitterUpdateScript [
   &EmitterState(Delay First Loop Only=true, Life Cycle Mode=::NewEnumerator1, Loop Delay=1, Loop Duration=1, UseLoopDelay=true)
   &SpawnRate(
    SpawnRate=&Lerp_Float(
     A=0
     B=&Power(a=$User.ParaGlideLeanAlpha, b=3)
     Alpha=&FN_PickFloatDependOnPlatform(TargetPlatformActive=50, TargetPlatformNoActive=25)
    )
    Use Spawn Probability=false
   )
   &SpawnRate(
    SpawnGroup=1
    SpawnRate=&Lerp_Float(
     A=0
     B=&Power(a=$User.ParaGlideLeanAlpha, b=3)
     Alpha=&FN_PickFloatDependOnPlatform(TargetPlatformActive=100, TargetPlatformNoActive=50)
    )
    Use Spawn Probability=false
   ) {module=SpawnRate001}
  ]
  particleSpawnScript [
   &InitializeParticle(
    Lifetime=&RandomRangeFloat(
     Minimum=&SelectFloatFromArray(Direct Array Index=&SpawnGroup(), Float Selection Array=DataInterface() {propertyValues {FloatData [0.4,1]}})
     Maximum=&SelectFloatFromArray(Direct Array Index=&SpawnGroup(), Float Selection Array=DataInterface() {propertyValues {FloatData [0.6,1.25]}})
    )
    Sprite Rotation Mode=::NewEnumerator2
    Sprite Size Mode=::NewEnumerator1
    Sprite UV Mode=::NewEnumerator3
    Uniform Sprite Size Max=100
    Uniform Sprite Size Min=75
    UsePositionOffset=false
    Write Material Random=true
   )
   &SubUVAnimation(
    Evaluation Type=::NewEnumerator0
    "Sprite Renderer "=DataInterface() {propertyValues {SpriteRenderer=&self:WaterSplashesDive_0.NiagaraSpriteRendererProperties_0}}
    SubUV Animation Mode:ENiagara_SubUVLookupModeV2=::NewEnumerator2
    UseEndFrame=false
    UseStartFrame=false
   )
   &SkeletalMeshLocation(
    Bone / Socket Index=4
    Bone Sampling Mode=::NewEnumerator4
    MaskSpawnGroup=false
    Sampled Position Offset=&MakeCustomVectorFromBool(
     Bool=$User.bIsFrontEndPreview
     True Vector=&RandomRangeVector(
      Minimum=&MakeVector(
       X=75
       Y=&SelectFloatFromArray(Direct Array Index=&SpawnGroup(), Float Selection Array=DataInterface() {propertyValues {FloatData [0,-50]}})
       Z=-75
      )
      Maximum=&MakeVector(X=-75, Y=-15, Z=-75)
     )
     False Vector=&RandomRangeVector(
      Minimum=&MakeVector(
       X=75
       Y=&SelectFloatFromArray(Direct Array Index=&SpawnGroup(), Float Selection Array=DataInterface() {propertyValues {FloatData [0,-50]}})
       Z=-200
      )
      Maximum=&MakeVector(X=-75, Y=-15, Z=-200)
     )
    )
    Sampled Position Offset Coordinate Space:ENiagaraSkelMeshTransforms=::NewEnumerator2
    Skeletal Mesh=$System.Skeletal Mesh
   )
   &AddVelocity(
    Velocity=&NegateVector(Vector:Vector3f=$System.FN_Contrail_Setup.Direction Vector)
    Velocity Speed Scale=&RandomRangeFloat(
     Minimum=&SelectFloatFromArray(Direct Array Index=&SpawnGroup(), Float Selection Array=DataInterface() {propertyValues {FloatData [500,600]}})
     Maximum=&SelectFloatFromArray(Direct Array Index=&SpawnGroup(), Float Selection Array=DataInterface() {propertyValues {FloatData [500,700]}})
    )
   )
  ]
  particleUpdateScript [
   &ParticleState()
   /SetVariables(Particles.SubSize:Vector2f=&Multiply_Vector2D(
     A=&Multiply_Vector2D_ByFloat(
      Float=&MakeCustomFloatFromBool(
       Bool=$User.bIsFrontEndPreview
       True Float=&RandomRangeFloat(Minimum=0.75, Maximum=1)
       False Float=&RandomRangeFloat(Minimum=0.75, Maximum=2)
      )
      Vector2D=<4 6>
     )
     B=&MakeCustomVector2DFromBool(
      Bool=$User.bIsFrontEndPreview
      True Vector2D=&Vector2DFromCurve(
       Vector2Curve=DataInterface() {propertyValues {
         XCurve {
          defaultValue=1
          keys [
           <at 0 0 RCIM_Cubic>
           {interpMode=RCIM_Cubic, tangentMode=RCTM_User, time=0.2, value=1}
           {arriveTangent=-2.0739, interpMode=RCIM_Cubic, leaveTangent=-2.0739, tangentMode=RCTM_User, time=1, value=0}
          ]
         }
         YCurve {
          defaultValue=1
          keys [
           {arriveTangent=10.1991, interpMode=RCIM_Cubic, leaveTangent=10.1991, tangentMode=RCTM_User, time=0.0338, value=0}
           {interpMode=RCIM_Cubic, tangentMode=RCTM_User, time=0.2, value=4}
           {arriveTangent=-5.214, interpMode=RCIM_Cubic, leaveTangent=-5.214, tangentMode=RCTM_User, time=0.5, value=1}
           {tangentMode=RCTM_User, time=1, value=0}
          ]
         }
        }}
       ScaleCurve=<1 1>
      )
      False Vector2D=&Vector2DFromCurve(
       Vector2Curve=DataInterface() {propertyValues {
         XCurve {
          defaultValue=1
          keys [
           {interpMode=RCIM_Cubic, tangentMode=RCTM_User, time=0, value=1}
           {arriveTangent=-2.0739, interpMode=RCIM_Cubic, leaveTangent=-2.0739, tangentMode=RCTM_User, time=1, value=0}
          ]
         }
         YCurve {
          defaultValue=1
          keys [
           {interpMode=RCIM_Cubic, tangentMode=RCTM_User, time=0, value=4}
           {arriveTangent=-5.214, interpMode=RCIM_Cubic, leaveTangent=-5.214, tangentMode=RCTM_User, time=0.5, value=1}
           {tangentMode=RCTM_User, time=1, value=0}
          ]
         }
        }}
       ScaleCurve=<1 1>
      )
     )
    )) {isSetParameters=true, module=SetVariables_#2}
   /SetVariables(Particles.VisibilityTag=&SpawnGroup()) {isSetParameters=true, module=SetVariables_#3}
   &ScaleSpriteSize(Uniform Curve Sprite Scale=DataInterface() {propertyValues {Curve {
       defaultValue=0
       keys [
        {arriveTangent=1.1866, interpMode=RCIM_Cubic, leaveTangent=1.1866, tangentMode=RCTM_User, time=0, value=0}
        {interpMode=RCIM_Cubic, tangentMode=RCTM_User, time=1, value=1}
       ]
      }}})
   &Drag(
    Drag=&SelectFloatFromArray(Direct Array Index=&SpawnGroup(), Float Selection Array=DataInterface() {propertyValues {FloatData [0,0.2]}})
    Use Linear Drag=true
    Use Rotational Drag=true
   )
   &ScaleColor(
    Scale Alpha=&FloatFromCurve(FloatCurve=DataInterface() {propertyValues {Curve {
        defaultValue=1
        keys [
         {arriveTangent=-0.1688, interpMode=RCIM_Cubic, leaveTangent=-0.1688, tangentMode=RCTM_User, time=0, value=1}
         {arriveTangent=-3.0312, interpMode=RCIM_Cubic, leaveTangent=-3.0312, tangentMode=RCTM_User, time=1, value=0}
        ]
       }}})
    ScaleA=true
    ScaleRGB=false
   )
   &DynamicMaterialParameters(
    Index 0 Param 3=&FloatFromCurve(FloatCurve=DataInterface() {propertyValues {Curve {
        defaultValue=1
        keys [{interpMode=RCIM_Cubic, tangentMode=RCTM_User, time=0, value=0}, {interpMode=RCIM_Cubic, tangentMode=RCTM_User, time=1, value=1}]
       }}})
    Index 0 Param 4=&FloatFromCurve(FloatCurve=DataInterface() {propertyValues {Curve {
        defaultValue=1
        keys [
         {interpMode=RCIM_Cubic, tangentMode=RCTM_User, time=0, value=0.6}
         {arriveTangent=1.2977, interpMode=RCIM_Cubic, leaveTangent=1.2977, tangentMode=RCTM_User, time=1, value=1}
        ]
       }}})
    Module.Index0Float0Write=false
    Module.Index0Float1Write=false
    Module.Index0Float2Write=true
    Module.Index0Float3Write=true
   )
   &GravityForce(Gravity=&SelectVectorFromArray(
     Direct Array Index=&SpawnGroup()
     Vector Selection Array=DataInterface() {propertyValues {FloatData [<0 0 0>, <0 0 -750>]}}
    ))
   &SolveForcesAndVelocity(Clamp Velocity=false, Limit Acceleration=false)
   &ScaleSpriteSizeBySpeed(Min Scale Factor=<0.5 0.5>)
  ]
  rendererClasses [&NiagaraSpriteRendererProperties]
  renderers [
   {rendererClass=&NiagaraSpriteRendererProperties}
   {rendererClass=&NiagaraSpriteRendererProperties, rendererIndex=1}
  ]
 }
 WaterSplashesXXL {
  emitterUpdateScript [
   &EmitterState(Delay First Loop Only=true, Life Cycle Mode=::NewEnumerator1, Loop Delay=1, Loop Duration=1, UseLoopDelay=true)
   &SpawnRate(SpawnRate=&Lerp_Float(A=10, B=0, Alpha=$User.ParaGlideLeanAlpha), Use Spawn Probability=false)
  ]
  particleSpawnScript [
   &InitializeParticle(
    Color=<rgba 1 1 1 0.9>
    Lifetime=0.7
    Sprite Rotation Mode=::NewEnumerator2
    Sprite Size Mode=::NewEnumerator1
    Sprite UV Mode=::NewEnumerator3
    Uniform Sprite Size Max=200
    Uniform Sprite Size Min=300
    UsePositionOffset=false
    Write Material Random=true
   )
   &SubUVAnimation(
    Evaluation Type=::NewEnumerator0
    "Sprite Renderer "=DataInterface() {propertyValues {SpriteRenderer=&self:WaterSplashesXXL_0.NiagaraSpriteRendererProperties_0}}
    SubUV Animation Mode:ENiagara_SubUVLookupModeV2=::NewEnumerator2
    UseEndFrame=false
    UseStartFrame=false
   )
   &SystemLocation(Offset=&MakeVector(X=&RandomRangeFloat(Minimum=-25, Maximum=25), Y=120, Z=0))
   &AddVelocity(Velocity=<0 -1 0>, Velocity Speed Scale=250)
  ]
  particleUpdateScript [
   &ParticleState()
   &ScaleSpriteSize(Uniform Curve Sprite Scale=DataInterface() {propertyValues {Curve {
       defaultValue=0
       keys [
        {arriveTangent=1.1866, interpMode=RCIM_Cubic, leaveTangent=1.1866, tangentMode=RCTM_User, time=0, value=0}
        {interpMode=RCIM_Cubic, tangentMode=RCTM_User, time=1, value=1}
       ]
      }}})
   &Drag(Drag=&Lerp_Float(A=1, B=0, Alpha=$User.ParaGlideLeanAlpha), Use Linear Drag=true, Use Rotational Drag=true)
   &ScaleColor(Scale Alpha=&Lerp_Float(A=1, B=0, Alpha=$User.ParaGlideLeanAlpha), ScaleA=true, ScaleRGB=false)
   &DynamicMaterialParameters(
    Index 0 Param 3=&FloatFromCurve(FloatCurve=DataInterface() {propertyValues {Curve {
        defaultValue=1
        keys [{interpMode=RCIM_Cubic, tangentMode=RCTM_User, time=0, value=0}, {interpMode=RCIM_Cubic, tangentMode=RCTM_User, time=1, value=1}]
       }}})
    Index 0 Param 4=&FloatFromCurve(FloatCurve=DataInterface() {propertyValues {Curve {
        defaultValue=1
        keys [
         {interpMode=RCIM_Cubic, tangentMode=RCTM_User, time=0, value=0.09}
         {arriveTangent=1.2977, interpMode=RCIM_Cubic, leaveTangent=1.2977, tangentMode=RCTM_User, time=1, value=1}
        ]
       }}})
    Module.Index0Float0Write=false
    Module.Index0Float1Write=false
    Module.Index0Float2Write=true
    Module.Index0Float3Write=true
   )
   &SolveForcesAndVelocity(Clamp Velocity=false, Limit Acceleration=false)
   &SpriteFacingAndAlignment(Facing Coordinate Space=::Local, Sprite Alignment=<0 0 1>, Sprite Facing=<0 0 1>, WriteAlignment=false, WriteFacing=true)
  ]
  rendererClasses [&NiagaraSpriteRendererProperties]
  renderers [{rendererClass=&NiagaraSpriteRendererProperties}]
 }
}
summary {
 systemName=NS_Contrail_BlueFriends
 emitters [
  {emitterName=WaterCalm, rendererClasses [&NiagaraMeshRendererProperties]}
  {emitterName=FishCalm, rendererClasses [&NiagaraMeshRendererProperties]}
  {emitterName=Ripples, rendererClasses [&NiagaraSpriteRendererProperties]}
  {emitterName=Splashes, rendererClasses [&NiagaraSpriteRendererProperties]}
  {emitterName=FishDive, rendererClasses [&NiagaraMeshRendererProperties]}
  {emitterName=WaterDive, rendererClasses [&NiagaraMeshRendererProperties]}
  {emitterName=WaterSplashesXXL, rendererClasses [&NiagaraSpriteRendererProperties]}
  {emitterName=WaterSplashesDive, rendererClasses [&NiagaraSpriteRendererProperties]}
  {emitterName=WaterSplashes, rendererClasses [&NiagaraSpriteRendererProperties]}
 ]
 userVariables [
  User.bIsFrontEnd {type=NiagaraBool}
  User.bIsFrontEndPreview {type=NiagaraBool}
  User.ParaGlideLeanAlpha {type=NiagaraFloat, defaultValue=0, description=""}
  User.RibbonIDOffset {type=NiagaraInt32}
 ]
}
